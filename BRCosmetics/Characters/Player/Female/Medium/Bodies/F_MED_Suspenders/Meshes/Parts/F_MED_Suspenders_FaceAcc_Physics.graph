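# Physics asset for the face accessory part of the female medium Suspenders cosmetic body, simulating its hair.
# Capsule chains hang from the head: three three-bone back-hair chains and two four-bone side strands, each with a small offshoot.
# Head, neck, upper spine and upper-arm capsules supply body shapes around the hair.
# Side-strand swing limits widen down the chain to free-swinging tips, while back-hair twist limits are mostly zero.
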PhysicsAsset /BRCosmetics/Characters/Player/Female/Medium/Bodies/F_MED_Suspenders/Meshes/Parts/F_MED_Suspenders_FaceAcc_Physics
bodies {
 dyn_C_backHair_A [{shapeType=Capsule, length=5, radius=2, shapeName=dyn_C_backHair_A_capsule, center=<3.3281 0 0.7386>, rotation=<rot 0 90 90>}]
 dyn_C_backHair_B [{shapeType=Capsule, length=11.75, radius=2.75, shapeName=dyn_C_backHair_B_capsule, center=<4.5074 0 0.0003>, rotation=<rot 0 1.1552 89.9998>}]
 dyn_C_backHair_C [{
   shapeType=Capsule
   length=9.5
   radius=3.5
   shapeName=dyn_C_backHair_C_capsule
   center=<3.8973 -0.9998 0.6736>
   rotation=<rot -11.0849 -17.3945 87.2394>
  }]
 dyn_L_backHair_A [{
   shapeType=Capsule
   length=6.75
   radius=2.5
   shapeName=dyn_R_backHair_A_capsule
   center=<5.5178 -0.5021 0.0538>
   rotation=<rot 0 89.9999 81.4792>
  }]
 dyn_L_backHair_B [{
   shapeType=Capsule
   length=4.25
   radius=2.5
   shapeName=dyn_L_backHair_B_capsule
   center=<1.8128 0.3546 -0.7999>
   rotation=<rot -0.0001 -3.1725 90>
  }]
 dyn_L_backHair_C [{shapeType=Capsule, length=5.75, radius=2.25, shapeName=dyn_L_backHair_C_capsule, center=<1.7016 0 0.2062>, rotation=<rot 0 89.9999 86.7602>}]
 dyn_L_hairStrand_A [{shapeType=Capsule, length=2.5, radius=1.5, shapeName=dyn_L_hairStrand_A_capsule, center=<-1 0 0>, rotation=<rot 90 0 0>}]
 dyn_L_hairStrand_B [{
   shapeType=Capsule
   length=2.75
   radius=1
   shapeName=dyn_L_hairStrand_B_capsule
   center=<-0.9898 0 0.4526>
   rotation=<rot 88.1326 0.0001 0.0001>
  }]
 dyn_L_hairStrand_C [{shapeType=Capsule, length=2.25, radius=0.75, shapeName=dyn_L_hairStrand_C_capsule, center=<-1 0 0>, rotation=<rot 90 0 0>}]
 dyn_L_hairStrand_D [{shapeType=Capsule, length=2.25, radius=0.75, shapeName=dyn_L_hairStrand_D_capsule, center=<-1 0 0>, rotation=<rot 90 -172.875 179.3862>}]
 dyn_L_smallHairStrand [{shapeType=Capsule, length=4, radius=1, shapeName=dyn_L_smallHairStrand_capsule, center=<-1 0 0>, rotation=<rot 90 0 0>}]
 dyn_R_backHair_A [{
   shapeType=Capsule
   length=6.75
   radius=2.5
   shapeName=dyn_R_backHair_A_capsule
   center=<5.5178 -0.5021 0.0538>
   rotation=<rot 0 89.9999 81.4792>
  }]
 dyn_R_backHair_B [{
   shapeType=Capsule
   length=1
   radius=3.5
   shapeName=dyn_R_backHair_B_capsule
   center=<0.8858 -1.5193 -0.494>
   rotation=<rot -0.0002 -3.1654 89.9999>
  }]
 dyn_R_backHair_C [{
   shapeType=Capsule
   length=6.5
   radius=1.75
   shapeName=dyn_R_backHair_C_capsule
   center=<0.0871 -2.067 -0.3199>
   rotation=<rot 5.2466 99.1923 79.6791>
  }]
 dyn_R_hairStrand_A [{shapeType=Capsule, length=3.75, radius=1.5, shapeName=dyn_R_hairStrand_A_capsule, center=<1 0 0>, rotation=<rot 90 0 0>}]
 dyn_R_hairStrand_B [{
   shapeType=Capsule
   length=2.25
   radius=1
   shapeName=dyn_R_hairStrand_B_capsule
   center=<0.9571 0 -0.4779>
   rotation=<rot 87.2588 -0.0001 -0.0001>
  }]
 dyn_R_hairStrand_C [{shapeType=Capsule, length=1.75, radius=0.75, shapeName=dyn_R_hairStrand_C_capsule, center=<1 0 0>, rotation=<rot 90 0 0>}]
 dyn_R_hairStrand_D [{shapeType=Capsule, length=2.25, radius=1, shapeName=dyn_R_hairStrand_D_capsule, center=<1 0 0>, rotation=<rot 90 2.3859 -20.9541>}]
 dyn_R_smallHairStrand [{shapeType=Capsule, length=4, radius=1, shapeName=dyn_R_smallHairStrand_capsule, center=<1 0 0>, rotation=<rot 90 0 0>}]
 head [{shapeType=Capsule, length=8.75, radius=6, shapeName=head_capsule, center=<0.5162 1.2416 0>, rotation=<rot 90 2.6838 0.904>}]
 neck_02 [
  {shapeType=Capsule, length=4.75, radius=3.25, shapeName=neck_02_capsule, center=<-0.6952 -0.5805 0>, rotation=<rot 90 0 0>}
  {shapeType=Capsule, length=25, radius=3.5, shapeName=neck_02_capsule}
 ]
 spine_05 [
  {shapeType=Capsule, length=16, radius=12.25, shapeName=spine_05_capsule, center=<0.7187 2.2735 0>}
  {shapeType=Capsule, length=15, radius=4.5, shapeName=spine_05_capsule, center=<15.0655 0.8215 -15.9192>, rotation=<rot 73.7689 0 6.074>}
  {shapeType=Capsule, length=15, radius=5, shapeName=spine_05_capsule, center=<13.8369 5.9622 15.9009>, rotation=<rot -61.1969 -2.8079 25.1816>}
 ]
 upperarm_l [{shapeType=Capsule, length=9.5, radius=5.5, shapeName=upperarm_l_capsule, center=<5.334 0.0001 0>, rotation=<rot 0 90 90>}]
 upperarm_r [{shapeType=Capsule, length=12.5, radius=5.5, shapeName=upperarm_r_capsule, center=<-6.1588 0.0002 0>, rotation=<rot 0 90 90>}]
}
constraints [
 {bone1Name=dyn_R_hairStrand_A, bone2Name=head, swing1LimitDegrees=10, swing2LimitDegrees=45, twistLimitDegrees=5, twistMotion=Locked}
 {bone1Name=dyn_R_hairStrand_B, bone2Name=dyn_R_hairStrand_A, swing1LimitDegrees=20, swing2LimitDegrees=45, twistLimitDegrees=5}
 {bone1Name=dyn_R_hairStrand_C, bone2Name=dyn_R_hairStrand_B, swing1LimitDegrees=35, swing2LimitDegrees=45, twistLimitDegrees=5}
 {
  bone1Name=dyn_R_hairStrand_D
  bone2Name=dyn_R_hairStrand_C
  swing1LimitDegrees=50
  swing1Motion=Free
  swing2LimitDegrees=50
  swing2Motion=Free
  twistLimitDegrees=5
  twistMotion=Locked
 }
 {bone1Name=dyn_R_smallHairStrand, bone2Name=dyn_R_hairStrand_B, swing1LimitDegrees=50, swing2LimitDegrees=50, twistLimitDegrees=5}
 {bone1Name=dyn_L_hairStrand_A, bone2Name=head, swing1LimitDegrees=10, swing2LimitDegrees=45, twistLimitDegrees=5, twistMotion=Locked}
 {bone1Name=dyn_L_hairStrand_B, bone2Name=dyn_L_hairStrand_A, swing1LimitDegrees=20, swing2LimitDegrees=45, twistLimitDegrees=5}
 {bone1Name=dyn_L_hairStrand_C, bone2Name=dyn_L_hairStrand_B, swing1LimitDegrees=35, swing2LimitDegrees=45, twistLimitDegrees=5}
 {
  bone1Name=dyn_L_hairStrand_D
  bone2Name=dyn_L_hairStrand_C
  swing1LimitDegrees=50
  swing1Motion=Free
  swing2LimitDegrees=50
  swing2Motion=Free
  twistLimitDegrees=5
  twistMotion=Locked
 }
 {
  bone1Name=dyn_L_smallHairStrand
  bone2Name=dyn_L_hairStrand_B
  swing1LimitDegrees=50
  swing2LimitDegrees=50
  twistLimitDegrees=5
  twistMotion=Locked
 }
 {bone1Name=head, bone2Name=neck_02, swing1LimitDegrees=45, swing2LimitDegrees=45, twistLimitDegrees=15}
 {
  bone1Name=neck_02
  bone2Name=spine_05
  swing1LimitDegrees=45
  swing1Motion=Free
  swing2LimitDegrees=45
  swing2Motion=Free
  twistLimitDegrees=45
  twistMotion=Free
 }
 {bone1Name=dyn_C_backHair_A, bone2Name=head, swing1LimitDegrees=15, swing2LimitDegrees=55, twistLimitDegrees=10}
 {bone1Name=dyn_C_backHair_B, bone2Name=dyn_C_backHair_A, swing1LimitDegrees=15, swing2LimitDegrees=45, twistLimitDegrees=0}
 {bone1Name=dyn_C_backHair_C, bone2Name=dyn_C_backHair_B, swing1LimitDegrees=15, swing2LimitDegrees=40, twistLimitDegrees=0}
 {bone1Name=dyn_R_backHair_A, bone2Name=head, swing1LimitDegrees=15, swing2LimitDegrees=55, twistLimitDegrees=0}
 {bone1Name=dyn_R_backHair_B, bone2Name=dyn_R_backHair_A, swing1LimitDegrees=20, swing2LimitDegrees=40, twistLimitDegrees=0}
 {bone1Name=dyn_R_backHair_C, bone2Name=dyn_R_backHair_B, swing1LimitDegrees=20, swing2LimitDegrees=40, twistLimitDegrees=0}
 {bone1Name=dyn_L_backHair_A, bone2Name=head, swing1LimitDegrees=15, swing2LimitDegrees=55, twistLimitDegrees=0}
 {bone1Name=dyn_L_backHair_B, bone2Name=dyn_L_backHair_A, swing1LimitDegrees=15, swing2LimitDegrees=45, twistLimitDegrees=0}
 {bone1Name=dyn_L_backHair_C, bone2Name=dyn_L_backHair_B, swing1LimitDegrees=8, swing2LimitDegrees=40, twistLimitDegrees=0}
 {
  bone1Name=upperarm_l
  bone2Name=spine_05
  swing1LimitDegrees=45
  swing1Motion=Free
  swing2LimitDegrees=45
  swing2Motion=Free
  twistLimitDegrees=45
  twistMotion=Free
 }
 {
  bone1Name=upperarm_r
  bone2Name=spine_05
  swing1LimitDegrees=45
  swing1Motion=Free
  swing2LimitDegrees=45
  swing2Motion=Free
  twistLimitDegrees=45
  twistMotion=Free
 }
]
