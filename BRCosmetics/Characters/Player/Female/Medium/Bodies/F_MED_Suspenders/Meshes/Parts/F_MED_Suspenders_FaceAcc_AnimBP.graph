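# Animation Blueprint for the F_MED_Suspenders face accessory, adding secondary hair motion on the player character.
# The pose copied from the attached parent mesh runs through RigidBody physics, spring-driven AnimDynamics on both buns, then additive back-hair bone rotations.
# Skydive floating lightens gravity and rotates the hair.
# While targeting with Alt_weapon set, head twist against spine_05 bends the hair; update sets that flag true without checking weapon type.
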AnimBlueprint /BRCosmetics/Characters/Player/Female/Medium/Bodies/F_MED_Suspenders/Meshes/Parts/F_MED_Suspenders_FaceAcc_AnimBP
target_skeleton=&/F_MED_Suspenders_FaceAcc_Skeleton
graphs [
 AnimGraph {nodes [
   n1 = RigidBody!(
    ComponentPose=Local To Component(LocalPose=Copy Pose From Mesh() {bUseAttachedParent=true})
    Alpha
    OverrideWorldGravity:Vector=Get skydivingGravity() -> (skydivingGravity:Vector)
    ExternalForce
   ) -> (Pose) {
    bClampLinearTranslationLimitToRefPose=true
    bOverrideWorldGravity=true
    overridePhysicsAsset=&/F_MED_Suspenders_FaceAcc_Physics
    simulationSpace=BaseBoneSpace
    alphaBoolBlend {blendOption=Linear}
    baseBoneRef {boneName=spine_05}
    componentLinearAccScale=<0.28 0.28 0.28>
    componentLinearVelScale=<0.28 0.28 0.28>
    overrideWorldGravity {z=-850}
   }
   n2 = Anim Dynamics - Bone: dyn_R_bun(ComponentPose=@n1, Alpha=1, ExternalForce="0,0,0") {
    angularDampingOverride=1
    angularSpringConstant=50
    bAngularSpring=true
    bGravityOverrideInSimSpace=true
    gravityScale=0.2
    alphaBoolBlend {blendOption=Linear}
    alphaScaleBiasClamp {scale=0.25}
    boundBone {boneName=dyn_R_bun}
    componentLinearAccScale=<0.5 0.5 0.5>
    physicsBodyDefinitions [{
      sphereCollisionRadius=0
      boundBone {boneName=dyn_R_bun}
      boxExtents=<3 3 3>
      constraintSetup {angularLimitsMax=<5 35 35>, angularLimitsMin=<-5 -35 -35>, angularTarget=<0 0 0>}
      localJointOffset=<0.1 0 0>
     }]
   }
   n3 = Anim Dynamics - Bone: dyn_L_bun(ComponentPose=@n2, Alpha=1, ExternalForce="0,0,0") {
    angularDampingOverride=1
    angularSpringConstant=50
    bAngularSpring=true
    bGravityOverrideInSimSpace=true
    gravityScale=0.2
    alphaBoolBlend {blendOption=Linear}
    alphaScaleBiasClamp {scale=0.25}
    boundBone {boneName=dyn_L_bun}
    componentLinearAccScale=<0.5 0.5 0.5>
    physicsBodyDefinitions [{
      sphereCollisionRadius=0
      boundBone {boneName=dyn_L_bun}
      boxExtents=<3 3 3>
      constraintSetup {angularLimitsMax=<5 35 35>, angularLimitsMin=<-5 -35 -35>, angularTarget=<0 0 0>}
      localJointOffset=<-0.1 0 0>
     }]
   }
   n4 = Get isSkydiveFloating() -> (isSkydiveFloating:double)
   n5 = "Transform (Modify) Bone - Bone: dyn_R_backHair_A"(ComponentPose=@n3, Alpha=@n4, Translation="0,0,0", Rotation="0,0,50", Scale="1,1,1") {
    rotationMode=BMM_Additive
    alphaBoolBlend {blendOption=Linear}
    alphaScaleBiasClamp {bInterpResult=true, scale=0.5}
    boneToModify {boneName=dyn_R_backHair_A}
    rotation {roll=50}
   }
   n6 = "Transform (Modify) Bone - Bone: dyn_L_backHair_A"(ComponentPose=@n5, Alpha=@n4, Translation="0,0,0", Rotation="0,0,50", Scale="1,1,1") {
    rotationMode=BMM_Additive
    alphaBoolBlend {blendOption=Linear}
    alphaScaleBiasClamp {bInterpResult=true, scale=0.5}
    boneToModify {boneName=dyn_L_backHair_A}
    rotation {roll=50}
   }
   n7 = "Transform (Modify) Bone - Bone: dyn_C_backHair_A"(ComponentPose=@n6, Alpha=@n4, Translation="0,0,0", Rotation="0,0,30", Scale="1,1,1") {
    rotationMode=BMM_Additive
    alphaBoolBlend {blendOption=Linear}
    alphaScaleBiasClamp {bInterpResult=true, scale=0.5}
    boneToModify {boneName=dyn_C_backHair_A}
    rotation {roll=30}
   }
   n8 = "Transform (Modify) Bone - Bone: dyn_R_backHair_B"(ComponentPose=@n7, Alpha=@n4, Translation="0,0,0", Rotation="0,0,50", Scale="1,1,1") {
    rotationMode=BMM_Additive
    alphaBoolBlend {blendOption=Linear}
    alphaScaleBiasClamp {bInterpResult=true, scale=0.5}
    boneToModify {boneName=dyn_R_backHair_B}
    rotation {roll=50}
   }
   n9 = "Transform (Modify) Bone - Bone: dyn_L_backHair_B"(ComponentPose=@n8, Alpha=@n4, Translation="0,0,0", Rotation="0,0,50", Scale="1,1,1") {
    rotationMode=BMM_Additive
    alphaBoolBlend {blendOption=Linear}
    alphaScaleBiasClamp {bInterpResult=true, scale=0.5}
    boneToModify {boneName=dyn_L_backHair_B}
    rotation {roll=50}
   }
   n10 = "Transform (Modify) Bone - Bone: dyn_C_backHair_C"(ComponentPose=@n9, Alpha=@n4, Translation="0,0,0", Rotation="0,0,30", Scale="1,1,1") {
    rotationMode=BMM_Additive
    alphaBoolBlend {blendOption=Linear}
    alphaScaleBiasClamp {bInterpResult=true, scale=0.5}
    boneToModify {boneName=dyn_C_backHair_C}
    rotation {roll=30}
   }
   n11 = Get headRotateVal() -> (headRotateVal:double)
   n12 = "Transform (Modify) Bone - Bone: dyn_C_backHair_B"(ComponentPose=@n10, Alpha=@n11, Translation="0,0,0", Rotation="-20,0,0", Scale="1,1,1") {
    rotationMode=BMM_Additive
    rotationSpace=BCS_ParentBoneSpace
    translationMode=BMM_Additive
    translationSpace=BCS_ParentBoneSpace
    alphaBoolBlend {blendOption=Linear}
    alphaScaleBiasClamp {interpSpeedDecreasing=4, interpSpeedIncreasing=6}
    boneToModify {boneName=dyn_C_backHair_B}
    rotation {pitch=-20}
   }
   n13 = "Transform (Modify) Bone - Bone: dyn_R_backHair_B"(ComponentPose=@n12, Alpha=@n11, Translation="0,0,0", Rotation="-20,0,0", Scale="1,1,1") {
    rotationMode=BMM_Additive
    rotationSpace=BCS_ParentBoneSpace
    translationMode=BMM_Additive
    translationSpace=BCS_ParentBoneSpace
    alphaBoolBlend {blendOption=Linear}
    alphaScaleBiasClamp {interpSpeedDecreasing=4, interpSpeedIncreasing=6}
    boneToModify {boneName=dyn_R_backHair_B}
    rotation {pitch=-20}
   }
   n14 = "Transform (Modify) Bone - Bone: dyn_R_backHair_A"(ComponentPose=@n13, Alpha=@n11, Translation="0,0,0", Rotation="25,-25,0", Scale="1,1,1") {
    rotationMode=BMM_Additive
    rotationSpace=BCS_ParentBoneSpace
    translationMode=BMM_Additive
    translationSpace=BCS_ParentBoneSpace
    alphaBoolBlend {blendOption=Linear}
    alphaScaleBiasClamp {interpSpeedDecreasing=4, interpSpeedIncreasing=6}
    boneToModify {boneName=dyn_R_backHair_A}
    rotation {pitch=25, yaw=-25}
   }
   Output Pose(Result=Component To Local(ComponentPose=@n14)) {layerGroup=None}
  ]}
 EventGraph {nodes [
   n1 = Cast To FortPlayerAnimInstance(Object=GetAnimInstance(self=GetOwnerSkeletalMeshComponent()))
   > Set Alt_weapon(Alt_weapon:bool=false) -> (Output_Get:bool)
   > n2 = Switch on EFortWeaponCoreAnimation!(execute, Selection:EFortWeaponCoreAnimation Enum=Get WeaponCoreAnim(self=@n1)) -> (
    Melee:Exec
    Pistol:Exec
    Shotgun:Exec
    PaperBlueprint:Exec
    Rifle:Exec
    MeleeOneHand:Exec
    MachinePistol:Exec
    RocketLauncher:Exec
    GrenadeLauncher:Exec
    GoingCommando:Exec
    AssaultRifle:Exec
    TacticalShotgun:Exec
    SniperRifle:Exec
    TrapPlacement:Exec
    ShoulderLauncher:Exec
    AbilityDecoTool:Exec
    Crossbow:Exec
    C4:Exec
    RemoteControl:Exec
    DualWield:Exec
    AR_BullPup:Exec
    AR_ForwardGrip:Exec
    MedPackPaddles:Exec
    SMG_P90:Exec
    AR_DrumGun:Exec
    Consumable_Small:Exec
    Consumable_Large:Exec
    Balloon:Exec
    MountedTurret:Exec
    CreativeTool:Exec
    ExplosiveBow:Exec
    AshtonIndigo:Exec
    AshtonChicago:Exec
    MeleeDualWield:Exec
    Unarmed:Exec
   )
   n3 = Branch(execute=[@n2.AssaultRifle, @n2.Rifle, @n2.Shotgun, @n2.TacticalShotgun])
   Set Alt_weapon(execute=[@n4.then_3, @n3.then], Alt_weapon=true)
   Event BlueprintUpdateAnimation()
   > n4 = Sequence!(execute) -> (then_0, then_1, then_2:Exec, then_3:Exec, then_4:Exec)
   n5 = Get bIsSkydiveFloating()
   > Set skydivingGravity(skydivingGravity:Vector=SelectVector(A="0, 0,-50", B="0, 0,-850", bPickA=@n5)) -> (Output_Get:Vector)
   Set isSkydiveFloating(execute=@n4.then_1, isSkydiveFloating:double="To Float (Boolean)"(InBool=@n5)) -> (Output_Get:double)
   n6 = Branch(execute=@n4.then_2, Condition=Get Alt_weapon() -> (Alt_weapon:bool))
   > n7 = Branch(Condition=Get bIsTargeting())
   > Set headRotateVal(headRotateVal:double=GetAlphaFromReferenceTransformDelta(
     SkeletalMeshComponent=Get bodyMesh() -> (bodyMesh:Skeletal Mesh Component Object Reference)
     BoneName=head
     BaseName=spine_05
     DeltaComponentType=QuaternionTwist
     InputRange=(X=0,Y=30)
     TwistAxis="0,0,1"
    )) -> (Output_Get:double)
   Set headRotateVal(execute=@n7.else, headRotateVal=0)
   Set headRotateVal(execute=@n6.else, headRotateVal=0)
   Event BlueprintInitializeAnimation()
   > Set bodyMesh(bodyMesh:Skeletal Mesh Component Object Reference=GetOwnerSkeletalMeshComponent()) -> (Output_Get:Skeletal Mesh Component Object Reference)
  ]}
]
variables [
 ADS_Alpha {type=boolean}
 Alt_weapon {type=boolean}
 bodyMesh {type=SkeletalMeshComponent}
 headRotateVal {type=number}
 isSkydiveFloating {type=number}
 skydivingGravity {type=Vector}
]
